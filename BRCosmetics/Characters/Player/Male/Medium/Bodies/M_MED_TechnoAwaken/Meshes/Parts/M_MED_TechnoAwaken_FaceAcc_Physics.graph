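# Physics asset for the face accessories of the M_MED_TechnoAwaken male character, covering left and right earrings and a lip ring bone.
# Each dyn_earring bone has a small sphere body, and the head has a large capsule.
# Three constraints tie the jewelry directly to the head with tight angular limits.
# Each has one looser axis for dangle: swing2 for the ear pieces, twist for the lip piece.
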PhysicsAsset /BRCosmetics/Characters/Player/Male/Medium/Bodies/M_MED_TechnoAwaken/Meshes/Parts/M_MED_TechnoAwaken_FaceAcc_Physics
bodies {
 dyn_earring_l [{shapeType=Sphere, radius=1, shapeName=dyn_earring_l_sphere, center=<-0.8143 0.0084 0.0091>}]
 dyn_earring_lip [{shapeType=Sphere, radius=0.947, shapeName=dyn_earring_lip_sphere, center=<-0.0129 0.0315 0.0014>}]
 dyn_earring_r [{shapeType=Sphere, radius=1, shapeName=dyn_earring_r_sphere, center=<-0.8273 0.0058 -0.0166>}]
 head [{shapeType=Capsule, length=9.1448, radius=8.9686, center=<8.7647 1.7757 0.0234>, rotation=<rot 0.1794 -80.9169 -90>}]
}
constraints [
 {bone1Name=dyn_earring_l, bone2Name=head, swing1LimitDegrees=2, swing2LimitDegrees=22, twistLimitDegrees=2}
 {bone1Name=dyn_earring_r, bone2Name=head, swing1LimitDegrees=2, swing2LimitDegrees=22, twistLimitDegrees=2}
 {bone1Name=dyn_earring_lip, bone2Name=head, swing1LimitDegrees=2, swing2LimitDegrees=2, twistLimitDegrees=20}
]
