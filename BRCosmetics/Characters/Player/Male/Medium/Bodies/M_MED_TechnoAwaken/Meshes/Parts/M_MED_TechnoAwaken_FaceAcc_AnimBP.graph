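# Post-process style Animation Blueprint for the TechnoAwaken male face accessory, following the attached parent mesh's pose and adding physics motion.
# A constant offset adjusts the corrective bone, while LOD 1 and LOD 2 enable extra additive translation and scale on dedicated LOD bones.
# Each update sets those LOD flags from the current LOD.
# A final RigidBody node simulates in head-bone space using an override physics asset and custom gravity.
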AnimBlueprint /BRCosmetics/Characters/Player/Male/Medium/Bodies/M_MED_TechnoAwaken/Meshes/Parts/M_MED_TechnoAwaken_FaceAcc_AnimBP
target_skeleton=&/M_MED_TechnoAwaken_FaceAcc_Skeleton
graphs [
 AnimGraph {nodes [
   n1 = "Transform (Modify) Bone - Bone: FaceAcc_Lod_Corrective"(
    ComponentPose=Local To Component(LocalPose=Copy Pose From Mesh() {bUseAttachedParent=true})
    Alpha=1
    Translation="0,-1,0"
    Rotation="0,0,0"
    Scale="1,1,1"
   ) {
    translationMode=BMM_Additive
    translationSpace=BCS_BoneSpace
    alphaBoolBlend {blendOption=Linear}
    boneToModify {boneName=FaceAcc_Lod_Corrective}
    translation {y=-1}
   }
   n2 = Get Is Lod 1() -> (Is Lod 1:bool)
   n3 = "Transform (Modify) Bone - Bone: FaceAcc_Lod_Corrective"!(ComponentPose=@n1, bAlphaBoolEnabled:bool=@n2, Translation="-0.570,-1.700,0.000", Rotation="0,0,0", Scale="1,1,1") -> (Pose) {
    alphaInputType=Bool
    translationMode=BMM_Additive
    translationSpace=BCS_BoneSpace
    alphaBoolBlend {blendOption=Linear}
    boneToModify {boneName=FaceAcc_Lod_Corrective}
    translation=<-0.57 -1.7>
   }
   n4 = "Transform (Modify) Bone - Bone: FaceAcc_Lod_Scale"!(ComponentPose=@n3, bAlphaBoolEnabled:bool=@n2, Translation="0,0,0", Rotation="0,0,0", Scale="1,1,1.3") -> (Pose) {
    alphaInputType=Bool
    scaleMode=BMM_Additive
    scaleSpace=BCS_BoneSpace
    alphaBoolBlend {blendOption=Linear}
    boneToModify {boneName=FaceAcc_Lod_Scale}
    scale {z=1.3}
   }
   n5 = Get Is Lod 2() -> (Is Lod 2:bool)
   n6 = "Transform (Modify) Bone - Bone: FaceAcc_Lod_Corrective"(ComponentPose=@n4, bAlphaBoolEnabled=@n5, Translation="-0.333,-7.710,-0.000", Rotation="0,0,0", Scale="1,1,1") {
    alphaInputType=Bool
    translationMode=BMM_Additive
    translationSpace=BCS_BoneSpace
    alphaBoolBlend {blendOption=Linear}
    boneToModify {boneName=FaceAcc_Lod_Corrective}
    translation=<-0.333 -7.71>
   }
   n7 = "Transform (Modify) Bone - Bone: FaceAcc_Lod_Scale"(ComponentPose=@n6, bAlphaBoolEnabled=@n5, Translation="0,0,0", Rotation="0,0,0", Scale="1.000,1.000,2.069") {
    alphaInputType=Bool
    scaleMode=BMM_Additive
    scaleSpace=BCS_BoneSpace
    alphaBoolBlend {blendOption=Linear}
    boneToModify {boneName=FaceAcc_Lod_Scale}
    scale {z=2.069}
   }
   n8 = RigidBody(ComponentPose=@n7) {
    bClampLinearTranslationLimitToRefPose=true
    bOverrideWorldGravity=true
    overridePhysicsAsset=&/M_MED_TechnoAwaken_FaceAcc_Physics
    simulationSpace=BaseBoneSpace
    alphaBoolBlend {blendOption=Linear}
    baseBoneRef {boneName=head}
    componentAppliedLinearAccClamp=<5000 5000 5000>
    overrideWorldGravity {z=-980}
    simSpaceSettings {maxAngularAcceleration=600, maxAngularVelocity=3, maxLinearAcceleration=500, maxLinearVelocity=400, velocityScaleZ=0, worldAlpha=0.65}
   }
   Output Pose(Result=Component To Local(ComponentPose=@n8)) {layerGroup=None}
  ]}
 EventGraph {nodes [
   Event BlueprintUpdateAnimation()
   n1 = Get CurrentLOD()
   > Set Is Lod 1(Is Lod 1:bool="Equal (Integer)"(A=@n1, B=1)) -> (Output_Get:bool)
   > Set Is Lod 2(Is Lod 2:bool="Equal (Integer)"(A=@n1, B=2)) -> (Output_Get:bool)
  ]}
]
variables [
 Is Lod 1 {type=boolean}
 Is Lod 2 {type=boolean}
]
